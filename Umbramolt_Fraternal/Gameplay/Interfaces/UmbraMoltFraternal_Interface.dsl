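; Blueprint interface for the UmbraMolt Fraternal gameplay, apparently a pair of twin pawns, Yuki and Gogo, in Fortnite.
; It declares setters and getters for intro speech, twin active state and walkout emote selection, plus a walkout-in-progress flag and post-walkout emote query.
; Other calls hand twin pawn and portal references between implementers and signal when spawned henchmen are eliminated.
; The file holds only function signatures with default returns, no logic.
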(fn GetIntroSpeechTag
  (return 0))

(fn GetPostWalkoutEmotes
  (return 0))

(fn GetTalkAtIntro
  (return false))

(fn GetTwinState
  (return false))

(fn GetWalkoutEmoteToPlay
  (return 0))

(fn PassBothTwinsReferences
  (params (YukiPawn :type Fort%20Pawn%20Object%20Reference) (GogoPawn :type Fort%20Pawn%20Object%20Reference)))

(fn PassPortalRef
  (params (Actor :type Actor%20Object%20Reference)))

(fn PassTwinReference
  (params (TwinPawn :type Fort%20Pawn%20Object%20Reference)))

(fn SetIsDoingWalkout
  (params (Bool :type Boolean)))

(fn SetTalkAtIntro
  (params (Bool :type Boolean)))

(fn SetWalkoutEmoteToPlay
  (params (Index :type Integer)))

(fn SpawnedHenchmenEliminated)

(fn ToggleTwinState
  (params (Active :type Boolean)))
